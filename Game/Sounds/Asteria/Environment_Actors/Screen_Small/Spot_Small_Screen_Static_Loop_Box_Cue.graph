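# Looping spot ambience cue for static on a small screen environment actor in Asteria, built from reused hologram-lock foley loops.
# A Random node picks one of four equally weighted looping waves and feeds a Modulator, with the cue's pitch multiplier doubled.
# A box-shaped attenuation override replaces the referenced capsule attenuation asset.
# A concurrency override prevents new instances at the limit, superseding the assigned weapon-drop stop-oldest set.
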SoundCue /Game/Sounds/Asteria/Environment_Actors/Screen_Small/Spot_Small_Screen_Static_Loop_Box_Cue
root=Modulator_0
nodes [
 {id=Modulator_0, children [Random_0], properties {pitchMin=1}}
 {id=Random_0, children [WavePlayer_1,WavePlayer_2,WavePlayer_3,WavePlayer_4], properties {weights [1,1,1,1]}}
 {
  id=WavePlayer_1
  properties {bLooping=true, soundWaveAssetPtr=/Game/Sounds/Foley_Loot/Containers/Keys_And_Locks/Keys_And_Locks_HologramLoop_01}
 }
 {
  id=WavePlayer_2
  properties {bLooping=true, soundWaveAssetPtr=/Game/Sounds/Foley_Loot/Containers/Keys_And_Locks/Keys_And_Locks_HologramLoop_02}
 }
 {
  id=WavePlayer_3
  properties {bLooping=true, soundWaveAssetPtr=/Game/Sounds/Foley_Loot/Containers/Keys_And_Locks/Keys_And_Locks_HologramLoop_03}
 }
 {
  id=WavePlayer_4
  properties {bLooping=true, soundWaveAssetPtr=/Game/Sounds/Foley_Loot/Containers/Keys_And_Locks/Keys_And_Locks_HologramLoop_04}
 }
]
properties {
 attenuationSettings=&/Spot_Small_Screen_Static_Capsule_Att
 bOverrideAttenuation=true
 bOverrideConcurrency=true
 maxDistance=655.649
 pitchMultiplier=2
 soundClassObject=&Fort_Gadgets_Items
 volumeMultiplier=1
 attenuationOverrides {attenuationShape=Box, falloffDistance=382.3265, attenuationShapeExtents=<127.5618 273.3225 250>}
 concurrencyOverrides {resolutionRule=PreventNew}
 concurrencySet [&Weapon_Drop_StopOldest]
}
