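# Sound cue for the looping static of a small screen environment actor in Asteria, built from reused hologram lock loops.
# A Random node picks one of four equally weighted looping waves and feeds a Modulator for slight variation, with the cue's pitch multiplier doubled.
# Only the first Quality Level input is connected.
# Spatialization uses a capsule attenuation asset, and a concurrency override that prevents new voices supersedes the assigned stop-oldest set.
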SoundCue /Game/Sounds/Asteria/Environment_Actors/Screen_Small/Spot_Small_Screen_Static_Loop_Cue
root=QualityLevel_0
nodes [
 {id=Modulator_0, children [Random_0], properties {pitchMin=1}}
 {id=QualityLevel_0, children [Modulator_0,null,null,null]}
 {id=Random_0, children [WavePlayer_1,WavePlayer_2,WavePlayer_3,WavePlayer_4], properties {weights [1,1,1,1]}}
 {
  id=WavePlayer_1
  properties {bLooping=true, soundWaveAssetPtr=/Game/Sounds/Foley_Loot/Containers/Keys_And_Locks/Keys_And_Locks_HologramLoop_01}
 }
 {
  id=WavePlayer_2
  properties {bLooping=true, soundWaveAssetPtr=/Game/Sounds/Foley_Loot/Containers/Keys_And_Locks/Keys_And_Locks_HologramLoop_02}
 }
 {
  id=WavePlayer_3
  properties {bLooping=true, soundWaveAssetPtr=/Game/Sounds/Foley_Loot/Containers/Keys_And_Locks/Keys_And_Locks_HologramLoop_03}
 }
 {
  id=WavePlayer_4
  properties {bLooping=true, soundWaveAssetPtr=/Game/Sounds/Foley_Loot/Containers/Keys_And_Locks/Keys_And_Locks_HologramLoop_04}
 }
]
properties {
 attenuationSettings=&/Spot_Small_Screen_Static_Capsule_Att
 bOverrideConcurrency=true
 maxDistance=854.7964
 pitchMultiplier=2
 soundClassObject=&Fort_Gadgets_Items
 volumeMultiplier=1
 attenuationOverrides {falloffDistance=2000}
 concurrencyOverrides {resolutionRule=PreventNew}
 concurrencySet [&Weapon_Drop_StopOldest]
}
